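# Idle glow Niagara system for the HonorBraceJoust pickaxe, built from mesh particles that follow a socket through a system-level socket reader.
# GlowMiddle grid-spawns a column of meshes, while GlowSides bursts mirrored side pieces placed by per-index offset, rotation and scale arrays.
# Execution index drives each piece's yellow-to-orange gradient, and the system Timeline fades alpha in.
# Front-end modifiers vary alpha, and an alternate mesh orientation is written for front-end display.
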NiagaraSystem /BRCosmetics/Weapons/FORT_Melee/Pickaxe_HonorBraceJoust/FX/NS_Pickaxe_HonorBraceJoust_Idle
dependencies {
 usedDataInterfaces [
  &NiagaraDataInterfaceArrayFloat3
  &NiagaraDataInterfaceCurve
  &NiagaraDataInterfaceColorCurve
  &NiagaraDataInterfaceSocketReader
 ]
 usedDynamicInputs [&InvertBool]
 usedModules [
  &self:SystemSpawnScript.SystemScriptSource.SystemScriptGraph.NiagaraNodeAssignment_0.SetVariables_#1
  &SystemState
  &Timeline
  &self:SystemSpawnScript.SystemScriptSource.SystemScriptGraph.NiagaraNodeAssignment_4.SetVariables_#2
 ]
}
emitters {
 GlowMiddle {
  emitterUpdateScript [
   &EmitterState(
    Inactive Response=::NewEnumerator1
    Life Cycle Mode=::NewEnumerator1
    Loop Behavior=::NewEnumerator1
    Loop Duration Mode=::NewEnumerator1
    UseLoopDelay=false
   )
   /Niagara/Modules/Spawn/Location/SpawnParticlesInGrid(X Count=1, Y Count=1, Z Count=12)
  ]
  particleSpawnScript [
   &InitializeParticle(
    Color=<rgba 55 55 55 1>
    Mesh Scale=&FN_FrontEndVector3Modifier(
     Gameplay=&SelectVectorFromArray(
      Direct Array Index=&ReturnExecIndex()
      Vector Selection Array=DataInterface() {propertyValues {FloatData [
         <0.15 0.1 0.1>
         <0.15 0.1 0.1>
         <0.15 0.1 0.1>
         <0.15 0.1 0.1>
         <0.15 0.1 0.1>
         <0.15 0.1 0.1>
         <0.15 0.1 0.1>
         <0.15 0.1 0.1>
         <0.15 0.2 0.1>
         <0.15 0.1 0.1>
         <0.15 0.1 0.1>
         <0.165 0.1 0.1>
        ]}}
     )
     Front End=&SelectVectorFromArray(
      Direct Array Index=&ReturnExecIndex()
      Vector Selection Array=DataInterface() {propertyValues {FloatData [
         <0.15 0.1 0.1>
         <0.15 0.1 0.1>
         <0.15 0.1 0.1>
         <0.15 0.1 0.1>
         <0.15 0.1 0.1>
         <0.15 0.1 0.1>
         <0.15 0.1 0.1>
         <0.15 0.1 0.1>
         <0.15 0.2 0.1>
         <0.15 0.1 0.1>
         <0.15 0.1 0.1>
         <0.165 0.1 0.1>
        ]}}
     )
    )
    Mesh Scale Mode=::NewEnumerator4
    UsePositionOffset=false
    Write Material Random=true
   )
   &InitialMeshOrientation(Mesh Orientation Mode=::NewEnumerator2, Rotation=<0 0.25 0>, Use Rotation Vector=true)
   /Niagara/Modules/Spawn/Location/V2/GridLocation(Module.Jitter Inside Grid Cells=false, Module.Override Random Seed=false, Offset:Vector3f=<0 0 0.5>)
  ]
  particleUpdateScript [
   &ParticleState(Kill Particles When Lifetime Has Elapsed=false, Let Infinitely Lived Particles Die When Emitter Deactivates=true)
   &SocketLocation(
    OverrideRandomSeed=false
    Position Sampling=::NewEnumerator1
    SamplePosOffset=true
    Sampled Position Offset=<5 5 0>
    Socket Reader=$System.Socket Reader
    Socket Sampling Mode=::NewEnumerator1
    VelocitySpeedLimit=false
   )
   &OffsetPosition(Position Offset:NiagaraPosition=$Output.SocketLocation.SampledPosition)
   &ScaleColor(
    Scale Alpha=&FN_FrontEndFloatModifier(
     Gameplay=&Multiply_Float(
      A=&FloatFromCurve(
       FloatCurve=DataInterface() {propertyValues {Curve {keys [
           {arriveTangent=1.2826, interpMode=RCIM_Cubic, leaveTangent=1.2826, tangentMode=RCTM_User, time=0, value=0}
           {tangentMode=RCTM_User, time=1, value=1}
          ]}}}
       Scale Curve=0.02
      )
      B=&Lerp_Float(A=0, B=1, Alpha=$System.Timeline.NormalizedTime)
     )
     Front End=&Multiply_Float(
      A=&FloatFromCurve(
       FloatCurve=DataInterface() {propertyValues {Curve {keys [
           {arriveTangent=1.2826, interpMode=RCIM_Cubic, leaveTangent=1.2826, tangentMode=RCTM_User, time=0, value=0}
           {tangentMode=RCTM_User, time=1, value=1}
          ]}}}
       Scale Curve=0.01
      )
      B=&Lerp_Float(A=0, B=1, Alpha=$System.Timeline.NormalizedTime)
     )
    )
    ScaleA=true
    ScaleRGB=true
   ) {module=ScaleColor002}
   &ScaleColor(
    Scale RGB=&MakeVectorFromLinearColorRGB(Color=&ColorFromCurve(
      ColorCurve=DataInterface() {propertyValues {
        AlphaCurve {keys [<at 0 1>]}
        BlueCurve {keys [<at 0 0.15>,<at 0.7011 0>]}
        GreenCurve {keys [<at 0 0.9047>,<at 0.7011 0.0156>]}
        RedCurve {keys [<at 0 0.9911>,<at 0.7011 1>]}
       }}
      CurveIndex=&ReturnNormalizedExecIndex()
     ))
    ScaleA=false
    ScaleRGB=true
   )
   /SetVariables(Particles.MeshOrientationFE:Quat4f=&LerpQuaternion(
     Quaternion A:Quat4f=$Particles.MeshOrientation
     Quaternion B:Quat4f=<0 1 0 1>
     Lerp Factor:NiagaraFloat=1
     Enable Spherical Linear Interpolation:NiagaraBool=false
    )) {isSetParameters=true, module=SetVariables_#3}
  ]
  rendererClasses [&NiagaraMeshRendererProperties]
  renderers [
   {rendererClass=&NiagaraMeshRendererProperties, rendererIndex=0}
   {rendererClass=&NiagaraMeshRendererProperties, rendererIndex=1}
  ]
 }
 GlowSides {
  emitterUpdateScript [
   &EmitterState(
    Inactive Response=::NewEnumerator1
    Life Cycle Mode=::NewEnumerator1
    Loop Behavior=::NewEnumerator1
    Loop Duration Mode=::NewEnumerator1
    UseLoopDelay=false
   )
   &SpawnBurst_Instantaneous(Spawn Count=6, Use Spawn Probability=false, UseLoopCountLimit=false)
  ]
  particleSpawnScript [
   &InitializeParticle(
    Color=<rgba 55 55 55 1>
    Mesh Scale=&FN_FrontEndVector3Modifier(
     Gameplay=&SelectVectorFromArray(
      Direct Array Index=&ReturnExecIndex()
      Vector Selection Array=DataInterface() {propertyValues {FloatData [<0.55 0.1 0.1>, <0.1 0.2 0.1>, <0.16 0.1 0.05>, <0.55 0.1 0.1>, <0.1 0.2 0.1>, <0.16 0.1 0.05>]}}
     )
     Front End=&SelectVectorFromArray(
      Direct Array Index=&ReturnExecIndex()
      Vector Selection Array=DataInterface() {propertyValues {FloatData [<0.55 0.1 0.1>, <0.1 0.2 0.1>, <0.16 0.1 0.05>, <0.55 0.1 0.1>, <0.1 0.2 0.1>, <0.16 0.1 0.05>]}}
     )
    )
    Mesh Scale Mode=::NewEnumerator4
    UsePositionOffset=false
    Write Material Random=true
   )
   &InitialMeshOrientation(
    Mesh Orientation Mode=::NewEnumerator2
    Rotation=&SelectVectorFromArray(
     Direct Array Index=&ReturnExecIndex()
     Vector Selection Array=DataInterface() {propertyValues {FloatData [<0.1 0.25 0>, <-0.1 0.25 0>, <0 0.25 0>, <-0.1 0.25 0>, <0.1 0.25 0>, <0 0.25 0>]}}
    )
    Use Rotation Vector=true
   )
  ]
  particleUpdateScript [
   &ParticleState(Kill Particles When Lifetime Has Elapsed=false, Let Infinitely Lived Particles Die When Emitter Deactivates=true)
   &SocketLocation(
    OverrideRandomSeed=false
    Position Sampling=::NewEnumerator1
    SamplePosOffset=true
    Sampled Position Offset=&SelectVectorFromArray(
     Direct Array Index=&ReturnExecIndex()
     Vector Selection Array=DataInterface() {propertyValues {FloatData [<0 12 43>, <0 12 27>, <0 12 17>, <0 -10 43>, <0 -10 27>, <0 -12 17>]}}
    )
    Socket Reader=$System.Socket Reader
    Socket Sampling Mode=::NewEnumerator1
    VelocitySpeedLimit=false
   )
   &OffsetPosition(Position Offset=&AddVectorToPosition(Position=$Output.SocketLocation.SampledPosition, Vector=<0 0 0>))
   &ScaleColor(
    Scale Alpha=&FN_FrontEndFloatModifier(
     Gameplay=&Multiply_Float(
      A=&FloatFromCurve(
       FloatCurve=DataInterface() {propertyValues {Curve {keys [
           {arriveTangent=1.2826, interpMode=RCIM_Cubic, leaveTangent=1.2826, tangentMode=RCTM_User, time=0, value=0}
           {tangentMode=RCTM_User, time=1, value=1}
          ]}}}
       Scale Curve=0.02
      )
      B=&Lerp_Float(A=0, B=1, Alpha=$System.Timeline.NormalizedTime)
     )
     Front End=&Multiply_Float(
      A=&FloatFromCurve(
       FloatCurve=DataInterface() {propertyValues {Curve {keys [
           {arriveTangent=1.2826, interpMode=RCIM_Cubic, leaveTangent=1.2826, tangentMode=RCTM_User, time=0, value=0}
           {tangentMode=RCTM_User, time=1, value=1}
          ]}}}
       Scale Curve=0.01
      )
      B=&Lerp_Float(A=0, B=1, Alpha=$System.Timeline.NormalizedTime)
     )
    )
    ScaleA=true
    ScaleRGB=true
   ) {module=ScaleColor002}
   &ScaleColor(
    Scale RGB=&MakeVectorFromLinearColorRGB(Color=&ColorFromCurve(
      ColorCurve=DataInterface() {propertyValues {
        AlphaCurve {keys [<at 0 1>]}
        BlueCurve {keys [<at 0.3 0>,<at 0.5 0.15>,<at 0.51 0>,<at 0.8 0>,<at 1 0.15>]}
        GreenCurve {keys [<at 0.3 0.0156>,<at 0.5 0.9047>,<at 0.51 0.0156>,<at 0.8 0.0156>,<at 1 0.9047>]}
        RedCurve {keys [<at 0.3 1>,<at 0.5 0.9911>,<at 0.51 1>,<at 0.8 1>,<at 1 0.9911>]}
       }}
      CurveIndex=&ReturnNormalizedExecIndex()
     ))
    ScaleA=false
    ScaleRGB=true
   )
   /SetVariables(Particles.MeshOrientationFE:Quat4f=&LerpQuaternion(
     Quaternion A:Quat4f=$Particles.MeshOrientation
     Quaternion B:Quat4f=<0 1 0 1>
     Lerp Factor:NiagaraFloat=1
     Enable Spherical Linear Interpolation:NiagaraBool=false
    )) {isSetParameters=true, module=SetVariables_#3}
  ]
  rendererClasses [&NiagaraMeshRendererProperties]
  renderers [
   {rendererClass=&NiagaraMeshRendererProperties, rendererIndex=0}
   {rendererClass=&NiagaraMeshRendererProperties, rendererIndex=1}
  ]
 }
}
summary {
 systemName=NS_Pickaxe_HonorBraceJoust_Idle
 emitters [
  {emitterName=GlowMiddle, rendererClasses [&NiagaraMeshRendererProperties]}
  {emitterName=GlowSides, rendererClasses [&NiagaraMeshRendererProperties]}
 ]
 userVariables [
  User.bIsFrontEnd {type=NiagaraBool}
  User.bIsFrontEndPreview {type=NiagaraBool}
  User.bisSwinging {type=NiagaraBool}
  User.FrontEndScale {type=NiagaraFloat, defaultValue=0.55, description=""}
  User.RibbonIDOffset {type=NiagaraInt32, defaultValue=1}
  User.SwingSectionIndex {type=NiagaraInt32}
 ]
}
